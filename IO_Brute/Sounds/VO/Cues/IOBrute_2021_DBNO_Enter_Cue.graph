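# Sound cue for the IO Brute NPC voice line played on entering the down-but-not-out (DBNO) state.
# A Random node picks one of six equally weighted wave variations and is excluded from branch culling.
# Playback uses the Fort_VO_NPC sound class, a close-range IO Brute attenuation asset, and the shared IO Brute chatter concurrency set.
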SoundCue /IO_Brute/Sounds/VO/Cues/IOBrute_2021_DBNO_Enter_Cue
root=Random_0
nodes [
 {
  id=Random_0
  children [WavePlayer_0,WavePlayer_1,WavePlayer_2,WavePlayer_3,WavePlayer_4,WavePlayer_5]
  properties {bShouldExcludeFromBranchCulling=true, weights [1,1,1,1,1,1]}
 }
 {id=WavePlayer_0, properties {soundWaveAssetPtr=/IO_Brute/Sounds/VO/Waves/IOBrute_2021_DBNO_Enter_010}}
 {id=WavePlayer_1, properties {soundWaveAssetPtr=/IO_Brute/Sounds/VO/Waves/IOBrute_2021_DBNO_Enter_011}}
 {id=WavePlayer_2, properties {soundWaveAssetPtr=/IO_Brute/Sounds/VO/Waves/IOBrute_2021_DBNO_Enter_012}}
 {id=WavePlayer_3, properties {soundWaveAssetPtr=/IO_Brute/Sounds/VO/Waves/IOBrute_2021_DBNO_Enter_013}}
 {id=WavePlayer_4, properties {soundWaveAssetPtr=/IO_Brute/Sounds/VO/Waves/IOBrute_2021_DBNO_Enter_014}}
 {id=WavePlayer_5, properties {soundWaveAssetPtr=/IO_Brute/Sounds/VO/Waves/IOBrute_2021_DBNO_Enter_015}}
]
properties {
 attenuationSettings=/IO_Brute/Sounds/IO_Brute_Close_Att
 maxDistance=4000
 soundClassObject=&Fort_VO_NPC
 volumeMultiplier=1
 concurrencySet [/IO_Brute/Sounds/IO_Brute_Chatter_Conc]
}
