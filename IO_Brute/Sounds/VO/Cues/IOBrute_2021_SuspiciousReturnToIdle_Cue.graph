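# Sound cue for the IO Brute NPC's voice lines, named for returning from suspicion to idle.
# A Random node picks among five equally weighted wave players, each holding one voice recording, and is excluded from branch culling.
# It plays in the NPC VO sound class, using the Brute's idle attenuation and chatter concurrency settings.
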SoundCue /IO_Brute/Sounds/VO/Cues/IOBrute_2021_SuspiciousReturnToIdle_Cue
root=Random_0
nodes [
 {
  id=Random_0
  children [WavePlayer_0,WavePlayer_1,WavePlayer_2,WavePlayer_3,WavePlayer_4]
  properties {bShouldExcludeFromBranchCulling=true, weights [1,1,1,1,1]}
 }
 {id=WavePlayer_0, properties {soundWaveAssetPtr=/IO_Brute/Sounds/VO/Waves/IOBrute_2021_SuspiciousReturnToIdle_010}}
 {id=WavePlayer_1, properties {soundWaveAssetPtr=/IO_Brute/Sounds/VO/Waves/IOBrute_2021_SuspiciousReturnToIdle_020}}
 {id=WavePlayer_2, properties {soundWaveAssetPtr=/IO_Brute/Sounds/VO/Waves/IOBrute_2021_SuspiciousReturnToIdle_030}}
 {id=WavePlayer_3, properties {soundWaveAssetPtr=/IO_Brute/Sounds/VO/Waves/IOBrute_2021_SuspiciousReturnToIdle_040}}
 {id=WavePlayer_4, properties {soundWaveAssetPtr=/IO_Brute/Sounds/VO/Waves/IOBrute_2021_SuspiciousReturnToIdle_050}}
]
properties {
 attenuationSettings=/IO_Brute/Sounds/IO_Brute_Idle_Att
 maxDistance=4000
 soundClassObject=&Fort_VO_NPC
 volumeMultiplier=1
 concurrencySet [/IO_Brute/Sounds/IO_Brute_Chatter_Conc]
}
