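# Sound cue for the IO Brute NPC's take-damage voice lines.
# A Random node picks one of eight equally weighted damage vocal waves and passes it through a Modulator at the root.
# The cue plays in the NPC VO sound class with a boosted volume, the Brute's close attenuation and a chatter concurrency set that governs overlapping lines.
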SoundCue /IO_Brute/Sounds/VO/Cues/IOBrute_2021_TakeDamage_Cue
root=Modulator_0
nodes [
 {id=Modulator_0, children [Random_1]}
 {
  id=Random_1
  children [WavePlayer_1, WavePlayer_4, WavePlayer_6, WavePlayer_7, WavePlayer_8, WavePlayer_11, WavePlayer_12, WavePlayer_9]
  properties {weights [1,1,1,1,1,1,1,1]}
 }
 {id=WavePlayer_1, properties {soundWaveAssetPtr=/IO_Brute/Sounds/VO/Waves/IOBrute_2021_TakeDamage_011}}
 {id=WavePlayer_11, properties {soundWaveAssetPtr=/IO_Brute/Sounds/VO/Waves/IOBrute_2021_TakeDamage_021}}
 {id=WavePlayer_12, properties {soundWaveAssetPtr=/IO_Brute/Sounds/VO/Waves/IOBrute_2021_TakeDamage_022}}
 {id=WavePlayer_4, properties {soundWaveAssetPtr=/IO_Brute/Sounds/VO/Waves/IOBrute_2021_TakeDamage_014}}
 {id=WavePlayer_6, properties {soundWaveAssetPtr=/IO_Brute/Sounds/VO/Waves/IOBrute_2021_TakeDamage_016}}
 {id=WavePlayer_7, properties {soundWaveAssetPtr=/IO_Brute/Sounds/VO/Waves/IOBrute_2021_TakeDamage_017}}
 {id=WavePlayer_8, properties {soundWaveAssetPtr=/IO_Brute/Sounds/VO/Waves/IOBrute_2021_TakeDamage_018}}
 {id=WavePlayer_9, properties {soundWaveAssetPtr=/IO_Brute/Sounds/VO/Waves/IOBrute_2021_TakeDamage_024}}
]
properties {
 attenuationSettings=/IO_Brute/Sounds/IO_Brute_Close_Att
 maxDistance=4000
 soundClassObject=&Fort_VO_NPC
 volumeMultiplier=2
 concurrencySet [/IO_Brute/Sounds/IO_Brute_Chatter_Conc]
}
