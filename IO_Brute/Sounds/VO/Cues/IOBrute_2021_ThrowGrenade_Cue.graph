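# Sound cue for the IO Brute NPC's voice line when throwing a grenade.
# A Random node picks one of three equally weighted wave recordings per play.
# It routes through the NPC VO sound class, with close-range attenuation and a chatter concurrency set shared by name with the Brute's sounds.
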SoundCue /IO_Brute/Sounds/VO/Cues/IOBrute_2021_ThrowGrenade_Cue
root=Random_0
nodes [
 {id=Random_0, children [WavePlayer_0,WavePlayer_1,WavePlayer_2], properties {weights [1,1,1]}}
 {id=WavePlayer_0, properties {soundWaveAssetPtr=/IO_Brute/Sounds/VO/Waves/IOBrute_2021_ThrowGrenade_010}}
 {id=WavePlayer_1, properties {soundWaveAssetPtr=/IO_Brute/Sounds/VO/Waves/IOBrute_2021_ThrowGrenade_011}}
 {id=WavePlayer_2, properties {soundWaveAssetPtr=/IO_Brute/Sounds/VO/Waves/IOBrute_2021_ThrowGrenade_012}}
]
properties {
 attenuationSettings=/IO_Brute/Sounds/IO_Brute_Close_Att
 maxDistance=4000
 soundClassObject=&Fort_VO_NPC
 volumeMultiplier=1
 concurrencySet [/IO_Brute/Sounds/IO_Brute_Chatter_Conc]
}
